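# Physics asset for the Sparkle Remix female medium body, setting up secondary motion for its skirt flaps, loincloth and creases.
# Mirrored left and right capsule chains have limited swing and twist, starting from small base bodies joined to the pelvis by angularly free constraints.
# Pelvis and thigh shapes provide body volumes for the cloth chains.
# Angularly free cross-links tie left to right on the loincloth and back flaps, and side flaps to their neighbouring skirt flaps.
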PhysicsAsset /BRCosmetics/Characters/Player/Female/Medium/Bodies/F_MED_SparkleRemix/Meshes/F_MED_Sparkle_Remix_Physics
bodies {
 dyn_backCrease_Base_l [{shapeType=Box, extentX=1.01, extentY=1.01, extentZ=1.01, rotation=<rot 0 -90 180>}]
 dyn_backCrease_Base_r [{shapeType=Box, extentX=1.01, extentY=1.01, extentZ=1.01, rotation=<rot 0 -90 0>}]
 dyn_backCrease_l [{
   shapeType=Capsule
   length=3.3187
   radius=1.084
   shapeName=dyn_backCrease_l_capsule
   center=<-2.5787 -0.0002 -0.0001>
   rotation=<rot -90 -0.005 0.005>
  }]
 dyn_backCrease_r [{
   shapeType=Capsule
   length=3.3074
   radius=1.023
   shapeName=dyn_backCrease_r_capsule
   center=<3.0846 0.0004 0>
   rotation=<rot -90 -0.0198 0.0198>
  }]
 dyn_crease_Base_l [{shapeType=Capsule, length=0.505, radius=0.505, shapeName=dyn_crease_Base_l_capsule, rotation=<rot 0 -90 0>}]
 dyn_crease_Base_r [{shapeType=Box, extentX=1.01, extentY=1.01, extentZ=1.01, rotation=<rot 0 90 90>}]
 dyn_crease_l [{shapeType=Capsule, length=2, radius=3, shapeName=dyn_crease_l_capsule, center=<-3 0 0>, rotation=<rot 25 0 0>}]
 dyn_crease_r [{shapeType=Capsule, length=3.5, radius=1.5, shapeName=dyn_crease_r_capsule, center=<3 0 0>, rotation=<rot -90 0 0>}]
 dyn_loincloth_01_Base_l [{shapeType=Box, extentX=1.01, extentY=1.01, extentZ=1.01, rotation=<rot 0 90 -90>}]
 dyn_loincloth_01_Base_r [{shapeType=Box, extentX=1.01, extentY=1.01, extentZ=1.01, rotation=<rot 0 90 90>}]
 dyn_loincloth_01_l [{shapeType=Capsule, length=2.5, radius=5, shapeName=dyn_loincloth_01_l_capsule, center=<-4.5 0 0>, rotation=<rot -90 2 -2>}]
 dyn_loincloth_01_r [{shapeType=Capsule, length=2.5, radius=3.5, shapeName=dyn_loincloth_01_l_capsule, center=<4.5 0 0>, rotation=<rot -90 -3.75 3.75>}]
 dyn_loincloth_02_l [{shapeType=Capsule, length=2.5, radius=4, shapeName=dyn_loincloth_02_l_capsule, center=<-4.5 0 0>, rotation=<rot -90 2 -2>}]
 dyn_loincloth_02_r [{shapeType=Capsule, length=2.5, radius=3, shapeName=dyn_loincloth_02_r_capsule, center=<4.8906 0.001 0>, rotation=<rot -90 -3.7927 3.7927>}]
 dyn_loincloth_03_l [{shapeType=Capsule, length=2.5, radius=3, shapeName=dyn_loincloth_03_l_capsule, center=<-4.5 0 0>, rotation=<rot -90 0 0>}]
 dyn_loincloth_03_r [{shapeType=Capsule, length=2.5, radius=3, shapeName=dyn_loincloth_03_r_capsule, center=<4.5 0 0>, rotation=<rot -90 0 0>}]
 dyn_loincloth_04_l [{shapeType=Capsule, length=2.5, radius=3, shapeName=dyn_loincloth_04_l_capsule, center=<-4.5 0 0>, rotation=<rot -90 0 0>}]
 dyn_loincloth_04_r [{shapeType=Capsule, length=2.5, radius=3, shapeName=dyn_loincloth_04_r_capsule, center=<4.5 0 0>, rotation=<rot -90 0 0>}]
 dyn_loincloth_05_l [{shapeType=Capsule, length=2.5, radius=3, shapeName=dyn_loincloth_05_l_capsule, center=<-4.5 0 0>, rotation=<rot -90 0 0>}]
 dyn_loincloth_05_r [{shapeType=Capsule, length=2.5, radius=3, shapeName=dyn_loincloth_05_r_capsule, center=<4.5 0 0>, rotation=<rot -90 0 0>}]
 dyn_skirtFlapBack_01_Base_l [{shapeType=Box, extentX=1.01, extentY=1.01, extentZ=1.01, rotation=<rot 0 -90 180>}]
 dyn_skirtFlapBack_01_Base_r [{shapeType=Box, extentX=1.01, extentY=1.01, extentZ=1.01, rotation=<rot 0 -90 0>}]
 dyn_skirtFlapBack_01_l [{shapeType=Capsule, length=3, radius=3, shapeName=dyn_skirtFlapBack_01_l_capsule, center=<-4 2 0>, rotation=<rot -90 0 0>}]
 dyn_skirtFlapBack_01_r [{shapeType=Capsule, length=2, radius=3, shapeName=dyn_skirtFlapBack_01_r_capsule, center=<4 -2 0>, rotation=<rot -90 0 0>}]
 dyn_skirtFlapBack_02_l [{shapeType=Capsule, length=3, radius=3, shapeName=dyn_skirtFlapBack_02_l_capsule, center=<-4 2 0>, rotation=<rot -90 0 0>}]
 dyn_skirtFlapBack_02_r [{shapeType=Capsule, length=2, radius=3, shapeName=dyn_skirtFlapBack_02_r_capsule, center=<4 -2 0>, rotation=<rot -90 0 0>}]
 dyn_skirtFlapBack_03_l [{shapeType=Capsule, length=3, radius=3, shapeName=dyn_skirtFlapBack_03_l_capsule, center=<-4 2 0>, rotation=<rot -90 0 0>}]
 dyn_skirtFlapBack_03_r [{shapeType=Capsule, length=2, radius=3, shapeName=dyn_skirtFlapBack_03_r_capsule, center=<4 -2 0>, rotation=<rot -90 0 0>}]
 dyn_skirtFlapBack_04_l [{shapeType=Capsule, length=3, radius=3, shapeName=dyn_skirtFlapBack_04_l_capsule, center=<-4 2 0>, rotation=<rot -90 0 0>}]
 dyn_skirtFlapBack_04_r [{shapeType=Capsule, length=2, radius=3, shapeName=dyn_skirtFlapBack_04_r_capsule, center=<4 -2 0>, rotation=<rot -90 0 0>}]
 dyn_skirtFlapSide_01_Base_l [{shapeType=Box, extentX=1.01, extentY=1.01, extentZ=1.01, rotation=<rot 0 -90 180>}]
 dyn_skirtFlapSide_01_Base_r [{shapeType=Box, extentX=1.01, extentY=1.01, extentZ=1.01, rotation=<rot 0 -90 0>}]
 dyn_skirtFlapSide_01_l [{shapeType=Capsule, length=2.5, radius=6, shapeName=dyn_skirtFlapSide_01_l_capsule, center=<-4 5 0>}]
 dyn_skirtFlapSide_01_r [{shapeType=Capsule, length=2.5, radius=6, shapeName=dyn_skirtFlapSide_01_l_capsule, center=<4 -4.5 0>}]
 dyn_skirtFlapSide_02_l [{shapeType=Capsule, length=2.5, radius=6, shapeName=dyn_skirtFlapSide_02_l_capsule, center=<-3 5 0>}]
 dyn_skirtFlapSide_02_r [{shapeType=Capsule, length=2.5, radius=6, shapeName=dyn_skirtFlapSide_02_l_capsule, center=<3 -4.5 0>}]
 dyn_skirtFlapSide_03_l [{shapeType=Capsule, length=2.5, radius=6, shapeName=dyn_skirtFlapSide_03_l_capsule, center=<-3 5 0>}]
 dyn_skirtFlapSide_03_r [{shapeType=Capsule, length=2.5, radius=6, shapeName=dyn_skirtFlapSide_03_l_capsule, center=<4.5 -4.5 0>}]
 dyn_skirtFlap_01_Base_l [{shapeType=Box, extentX=1.01, extentY=1.01, extentZ=1.01, rotation=<rot 0 90 -90>}]
 dyn_skirtFlap_01_Base_r [{shapeType=Box, extentX=1.01, extentY=1.01, extentZ=1.01, rotation=<rot 0 90 90>}]
 dyn_skirtFlap_01_l [{shapeType=Capsule, length=8, radius=6, shapeName=dyn_skirtFlap_01_l_capsule, center=<-2.75 5 0>, rotation=<rot 0 0 16>}]
 dyn_skirtFlap_01_r [{shapeType=Capsule, length=9, radius=6, shapeName=dyn_skirtFlap_01_l_capsule, center=<2 -4 1.25>}]
 dyn_skirtFlap_02_l [{shapeType=Capsule, length=4.75, radius=6, shapeName=dyn_skirtFlap_03_l_capsule, center=<-4.75 5 0>}]
 dyn_skirtFlap_02_r [{shapeType=Capsule, length=4.7239, radius=6, shapeName=dyn_skirtFlap_03_l_capsule, center=<4.25 -4 0>}]
 dyn_skirtFlap_03_l [{shapeType=Capsule, length=5.2239, radius=6, shapeName=dyn_skirtFlap_03_l_capsule, center=<-4 5 0>}]
 dyn_skirtFlap_03_r [{shapeType=Capsule, length=5.2239, radius=6, shapeName=dyn_skirtFlap_03_l_capsule, center=<3.25 -4.5 0>}]
 dyn_skirtFlap_04_l [{shapeType=Capsule, length=5.7239, radius=6, shapeName=dyn_skirtFlap_03_l_capsule, center=<-3.5 5 0>}]
 dyn_skirtFlap_04_r [{shapeType=Capsule, length=5.7, radius=6, shapeName=dyn_skirtFlap_03_l_capsule, center=<3 -4.5 0>}]
 pelvis [
  {shapeType=Sphere, radius=10.0092, shapeName=PelvisRight, center=<-3.9273 -2.335 5.8105>}
  {shapeType=Sphere, radius=10.009, shapeName=PelvisLeft, center=<-3.9273 -2.335 -5.8105>}
  {shapeType=Sphere, radius=7.8763, center=<-2.7785 -0.3675 6.4514>}
  {shapeType=Capsule, length=8.569, radius=9.922, shapeName=PelvisMiddle, center=<-4.9319 -2.3748 0.5301>, rotation=<rot 0 -90 0>}
 ]
 thigh_l [{shapeType=Capsule, length=14, radius=8, center=<-8 -1 0>, rotation=<rot 90 0 0>}]
 thigh_r [{shapeType=Capsule, length=5.5, radius=9.5, center=<13.6362 0.5321 -1.0032>, rotation=<rot 0 -93.1675 -88.5603>}]
 thigh_twist_01_l [
  {shapeType=Capsule, length=21.8971, radius=6.2332, center=<13.2416 -0.0003 0>, rotation=<rot 90 0 180>}
  {shapeType=Capsule, length=15.9933, radius=5, center=<20.4742 2.0482 1.0293>, rotation=<rot 90 0 -176.5764>}
 ]
 thigh_twist_01_r [
  {shapeType=Capsule, length=21.8971, radius=6.2332, center=<-13.1209 0.0001 0>, rotation=<rot 90 0 -179.9999>}
  {shapeType=Capsule, length=15.9933, radius=5, center=<-19.5158 4.4409 1.0293>, rotation=<rot 90 0 -176.5764>}
 ]
}
constraints [
 {bone1Name=dyn_skirtFlapBack_01_r, bone2Name=dyn_skirtFlapBack_01_Base_r, swing1LimitDegrees=20, swing2LimitDegrees=20, twistLimitDegrees=12}
 {bone1Name=dyn_skirtFlapBack_02_r, bone2Name=dyn_skirtFlapBack_01_r, swing1LimitDegrees=20, swing2LimitDegrees=20, twistLimitDegrees=12}
 {bone1Name=dyn_skirtFlapBack_03_r, bone2Name=dyn_skirtFlapBack_02_r, swing1LimitDegrees=20, swing2LimitDegrees=20, twistLimitDegrees=12}
 {bone1Name=dyn_skirtFlapBack_04_r, bone2Name=dyn_skirtFlapBack_03_r, swing1LimitDegrees=20, swing2LimitDegrees=20, twistLimitDegrees=12}
 {bone1Name=dyn_skirtFlapSide_01_r, bone2Name=dyn_skirtFlapSide_01_Base_r, swing1LimitDegrees=25, swing2LimitDegrees=25, twistLimitDegrees=12}
 {bone1Name=dyn_skirtFlapSide_02_r, bone2Name=dyn_skirtFlapSide_01_r, swing1LimitDegrees=25, swing2LimitDegrees=25, twistLimitDegrees=12}
 {bone1Name=dyn_skirtFlapSide_03_r, bone2Name=dyn_skirtFlapSide_02_r, swing1LimitDegrees=25, swing2LimitDegrees=25, twistLimitDegrees=12}
 {bone1Name=dyn_skirtFlap_01_r, bone2Name=dyn_skirtFlap_01_Base_r, swing1LimitDegrees=25, swing2LimitDegrees=25, twistLimitDegrees=12}
 {bone1Name=dyn_skirtFlap_02_r, bone2Name=dyn_skirtFlap_01_r, swing1LimitDegrees=25, swing2LimitDegrees=25, twistLimitDegrees=12}
 {bone1Name=dyn_skirtFlap_03_r, bone2Name=dyn_skirtFlap_02_r, swing1LimitDegrees=25, swing2LimitDegrees=25, twistLimitDegrees=12}
 {bone1Name=dyn_skirtFlap_04_r, bone2Name=dyn_skirtFlap_03_r, swing1LimitDegrees=25, swing2LimitDegrees=25, twistLimitDegrees=12}
 {bone1Name=dyn_backCrease_r, bone2Name=dyn_backCrease_Base_r, swing1LimitDegrees=20, swing2LimitDegrees=20, twistLimitDegrees=12}
 {bone1Name=dyn_crease_r, bone2Name=dyn_crease_Base_r, swing1LimitDegrees=25, swing2LimitDegrees=25, twistLimitDegrees=12}
 {bone1Name=dyn_loincloth_01_r, bone2Name=dyn_loincloth_01_Base_r, swing1LimitDegrees=25, swing2LimitDegrees=25, twistLimitDegrees=12}
 {bone1Name=dyn_loincloth_02_r, bone2Name=dyn_loincloth_01_r, swing1LimitDegrees=25, swing2LimitDegrees=25, twistLimitDegrees=12}
 {bone1Name=dyn_loincloth_03_r, bone2Name=dyn_loincloth_02_r, swing1LimitDegrees=25, swing2LimitDegrees=25, twistLimitDegrees=12}
 {bone1Name=dyn_loincloth_04_r, bone2Name=dyn_loincloth_03_r, swing1LimitDegrees=25, swing2LimitDegrees=25, twistLimitDegrees=12}
 {bone1Name=dyn_loincloth_05_r, bone2Name=dyn_loincloth_04_r, swing1LimitDegrees=25, swing2LimitDegrees=25, twistLimitDegrees=12}
 {bone1Name=dyn_loincloth_01_l, bone2Name=dyn_loincloth_01_Base_l, swing1LimitDegrees=25, swing2LimitDegrees=25, twistLimitDegrees=12}
 {bone1Name=dyn_loincloth_02_l, bone2Name=dyn_loincloth_01_l, swing1LimitDegrees=25, swing2LimitDegrees=25, twistLimitDegrees=12}
 {bone1Name=dyn_loincloth_03_l, bone2Name=dyn_loincloth_02_l, swing1LimitDegrees=25, swing2LimitDegrees=25, twistLimitDegrees=12}
 {bone1Name=dyn_loincloth_04_l, bone2Name=dyn_loincloth_03_l, swing1LimitDegrees=25, swing2LimitDegrees=25, twistLimitDegrees=12}
 {bone1Name=dyn_loincloth_05_l, bone2Name=dyn_loincloth_04_l, swing1LimitDegrees=25, swing2LimitDegrees=25, twistLimitDegrees=12}
 {
  bone1Name=dyn_loincloth_04_l
  bone2Name=dyn_loincloth_04_r
  swing1LimitDegrees=45
  swing1Motion=Free
  swing2LimitDegrees=45
  swing2Motion=Free
  twistLimitDegrees=45
  twistMotion=Free
 }
 {bone1Name=dyn_skirtFlap_01_l, bone2Name=dyn_skirtFlap_01_Base_l, swing1LimitDegrees=25, swing2LimitDegrees=25, twistLimitDegrees=12}
 {bone1Name=dyn_skirtFlap_02_l, bone2Name=dyn_skirtFlap_01_l, swing1LimitDegrees=25, swing2LimitDegrees=25, twistLimitDegrees=12}
 {bone1Name=dyn_skirtFlap_03_l, bone2Name=dyn_skirtFlap_02_l, swing1LimitDegrees=25, swing2LimitDegrees=25, twistLimitDegrees=12}
 {bone1Name=dyn_skirtFlap_04_l, bone2Name=dyn_skirtFlap_03_l, swing1LimitDegrees=25, swing2LimitDegrees=25, twistLimitDegrees=12}
 {bone1Name=dyn_skirtFlapSide_01_l, bone2Name=dyn_skirtFlapSide_01_Base_l, swing1LimitDegrees=25, swing2LimitDegrees=25, twistLimitDegrees=12}
 {bone1Name=dyn_skirtFlapSide_02_l, bone2Name=dyn_skirtFlapSide_01_l, swing1LimitDegrees=25, swing2LimitDegrees=25, twistLimitDegrees=12}
 {bone1Name=dyn_skirtFlapSide_03_l, bone2Name=dyn_skirtFlapSide_02_l, swing1LimitDegrees=25, swing2LimitDegrees=25, twistLimitDegrees=12}
 {bone1Name=dyn_backCrease_l, bone2Name=dyn_backCrease_Base_l, swing1LimitDegrees=20, swing2LimitDegrees=20, twistLimitDegrees=12}
 {bone1Name=dyn_skirtFlapBack_01_l, bone2Name=dyn_skirtFlapBack_01_Base_l, swing1LimitDegrees=20, swing2LimitDegrees=20, twistLimitDegrees=12}
 {bone1Name=dyn_skirtFlapBack_02_l, bone2Name=dyn_skirtFlapBack_01_l, swing1LimitDegrees=20, swing2LimitDegrees=20, twistLimitDegrees=12}
 {bone1Name=dyn_skirtFlapBack_03_l, bone2Name=dyn_skirtFlapBack_02_l, swing1LimitDegrees=20, swing2LimitDegrees=20, twistLimitDegrees=12}
 {bone1Name=dyn_skirtFlapBack_04_l, bone2Name=dyn_skirtFlapBack_03_l, swing1LimitDegrees=20, swing2LimitDegrees=20, twistLimitDegrees=12}
 {
  bone1Name=dyn_skirtFlapBack_03_r
  bone2Name=dyn_skirtFlapBack_03_l
  swing1LimitDegrees=45
  swing1Motion=Free
  swing2LimitDegrees=45
  swing2Motion=Free
  twistLimitDegrees=45
  twistMotion=Free
 }
 {
  bone1Name=thigh_twist_01_r
  bone2Name=thigh_r
  swing1LimitDegrees=45
  swing1Motion=Free
  swing2LimitDegrees=45
  swing2Motion=Free
  twistLimitDegrees=45
  twistMotion=Free
 }
 {
  bone1Name=dyn_backCrease_Base_r
  bone2Name=pelvis
  swing1LimitDegrees=45
  swing1Motion=Free
  swing2LimitDegrees=45
  swing2Motion=Free
  twistLimitDegrees=45
  twistMotion=Free
 }
 {
  bone1Name=dyn_backCrease_Base_l
  bone2Name=pelvis
  swing1LimitDegrees=45
  swing1Motion=Free
  swing2LimitDegrees=45
  swing2Motion=Free
  twistLimitDegrees=45
  twistMotion=Free
 }
 {
  bone1Name=dyn_skirtFlapBack_01_Base_l
  bone2Name=pelvis
  swing1LimitDegrees=45
  swing1Motion=Free
  swing2LimitDegrees=45
  swing2Motion=Free
  twistLimitDegrees=45
  twistMotion=Free
 }
 {
  bone1Name=dyn_skirtFlapBack_01_Base_r
  bone2Name=pelvis
  swing1LimitDegrees=45
  swing1Motion=Free
  swing2LimitDegrees=45
  swing2Motion=Free
  twistLimitDegrees=45
  twistMotion=Free
 }
 {
  bone1Name=dyn_skirtFlapSide_01_Base_r
  bone2Name=pelvis
  swing1LimitDegrees=45
  swing1Motion=Free
  swing2LimitDegrees=45
  swing2Motion=Free
  twistLimitDegrees=45
  twistMotion=Free
 }
 {
  bone1Name=dyn_skirtFlap_01_Base_r
  bone2Name=pelvis
  swing1LimitDegrees=45
  swing1Motion=Free
  swing2LimitDegrees=45
  swing2Motion=Free
  twistLimitDegrees=45
  twistMotion=Free
 }
 {
  bone1Name=dyn_crease_Base_r
  bone2Name=pelvis
  swing1LimitDegrees=45
  swing1Motion=Free
  swing2LimitDegrees=45
  swing2Motion=Free
  twistLimitDegrees=45
  twistMotion=Free
 }
 {
  bone1Name=dyn_loincloth_01_Base_r
  bone2Name=pelvis
  swing1LimitDegrees=45
  swing1Motion=Free
  swing2LimitDegrees=45
  swing2Motion=Free
  twistLimitDegrees=45
  twistMotion=Free
 }
 {
  bone1Name=dyn_skirtFlapSide_01_Base_l
  bone2Name=pelvis
  swing1LimitDegrees=45
  swing1Motion=Free
  swing2LimitDegrees=45
  swing2Motion=Free
  twistLimitDegrees=45
  twistMotion=Free
 }
 {
  bone1Name=dyn_skirtFlap_01_Base_l
  bone2Name=pelvis
  swing1LimitDegrees=45
  swing1Motion=Free
  swing2LimitDegrees=45
  swing2Motion=Free
  twistLimitDegrees=45
  twistMotion=Free
 }
 {
  bone1Name=dyn_loincloth_01_Base_l
  bone2Name=pelvis
  swing1LimitDegrees=45
  swing1Motion=Free
  swing2LimitDegrees=45
  swing2Motion=Free
  twistLimitDegrees=45
  twistMotion=Free
 }
 {
  bone1Name=thigh_l
  bone2Name=pelvis
  swing1LimitDegrees=45
  swing1Motion=Free
  swing2LimitDegrees=45
  swing2Motion=Free
  twistLimitDegrees=45
  twistMotion=Free
 }
 {
  bone1Name=thigh_r
  bone2Name=pelvis
  swing1LimitDegrees=45
  swing1Motion=Free
  swing2LimitDegrees=45
  swing2Motion=Free
  twistLimitDegrees=45
  twistMotion=Free
 }
 {
  bone1Name=dyn_skirtFlap_03_r
  bone2Name=dyn_skirtFlapSide_03_r
  swing1LimitDegrees=45
  swing1Motion=Free
  swing2LimitDegrees=45
  swing2Motion=Free
  twistLimitDegrees=45
  twistMotion=Free
 }
 {
  bone1Name=dyn_skirtFlapSide_03_l
  bone2Name=dyn_skirtFlap_03_l
  swing1LimitDegrees=45
  swing1Motion=Free
  swing2LimitDegrees=45
  swing2Motion=Free
  twistLimitDegrees=45
  twistMotion=Free
 }
 {
  bone1Name=thigh_twist_01_l
  bone2Name=thigh_l
  swing1LimitDegrees=45
  swing1Motion=Free
  swing2LimitDegrees=45
  swing2Motion=Free
  twistLimitDegrees=45
  twistMotion=Free
 }
 {
  bone1Name=dyn_skirtFlapBack_04_r
  bone2Name=dyn_skirtFlapBack_04_l
  swing1LimitDegrees=45
  swing1Motion=Free
  swing2LimitDegrees=45
  swing2Motion=Free
  twistLimitDegrees=45
  twistMotion=Free
 }
 {
  bone1Name=dyn_skirtFlapSide_03_r
  bone2Name=dyn_skirtFlap_04_r
  swing1LimitDegrees=45
  swing1Motion=Free
  swing2LimitDegrees=45
  swing2Motion=Free
  twistLimitDegrees=45
  twistMotion=Free
 }
 {
  bone1Name=dyn_skirtFlapSide_03_l
  bone2Name=dyn_skirtFlap_04_l
  swing1LimitDegrees=45
  swing1Motion=Free
  swing2LimitDegrees=45
  swing2Motion=Free
  twistLimitDegrees=45
  twistMotion=Free
 }
 {
  bone1Name=dyn_crease_Base_l
  bone2Name=pelvis
  swing1LimitDegrees=45
  swing1Motion=Free
  swing2LimitDegrees=45
  swing2Motion=Free
  twistLimitDegrees=45
  twistMotion=Free
 }
 {bone1Name=dyn_crease_l, bone2Name=dyn_crease_Base_l, swing1LimitDegrees=25, swing2LimitDegrees=25, twistLimitDegrees=12}
]
