AnimBlueprint /BRCosmetics/Characters/Player/Female/Medium/Bodies/F_MED_SparkleRemix/Meshes/F_MED_Sparkle_Remix_AnimBP
target_skeleton=&/F_MED_Sparkle_Remix_Skeleton
graphs [
 AnimGraph {nodes [
   n1 = RigidBody(ComponentPose=Local To Component(LocalPose=Use cached pose 'preRigids'())) {
    bClampLinearTranslationLimitToRefPose=true
    bOverrideWorldGravity=true
    overridePhysicsAsset=&/F_MED_Sparkle_Remix_Physics
    simulationSpace=BaseBoneSpace
    alphaBoolBlend {blendOption=Linear}
    baseBoneRef {boneName=pelvis}
    componentAppliedLinearAccClamp=<1500 1500 2000>
    componentLinearAccScale=<0.35 0.35 0.375>
    componentLinearVelScale=<0.5 0.5 0.5>
    overrideWorldGravity {z=-980}
   }
   n2 = Get crouchAlpha() -> (crouchAlpha:bool)
   n3 = "Transform (Modify) Bone - Bone: dyn_skirtFlap_01_Base_l"!(ComponentPose=@n1, bAlphaBoolEnabled:bool=@n2, Translation="1,0,0", Rotation="0,0,10", Scale="1,1,1") -> (Pose) {
    alphaInputType=Bool
    rotationMode=BMM_Additive
    rotationSpace=BCS_BoneSpace
    translationMode=BMM_Additive
    translationSpace=BCS_BoneSpace
    alphaBoolBlend {blendOption=Linear}
    boneToModify {boneName=dyn_skirtFlap_01_Base_l}
    rotation {roll=10}
    translation {x=1}
   }
   n4 = "Transform (Modify) Bone - Bone: dyn_crease_l"!(ComponentPose=@n3, bAlphaBoolEnabled:bool=@n2, Translation="3,0,1", Rotation="0,0,0", Scale="1,1,1") -> (Pose) {
    alphaInputType=Bool
    rotationMode=BMM_Additive
    rotationSpace=BCS_BoneSpace
    translationMode=BMM_Additive
    translationSpace=BCS_BoneSpace
    alphaBoolBlend {blendOption=Linear}
    boneToModify {boneName=dyn_crease_l}
    translation {x=3, z=1}
   }
   n5 = "integer > integer"(A=Get CurrentLOD(), B=1) {is=title K2Node_CallFunction}
   n6 = NOT Boolean(A=Get bRigidBodyEnabled())
   n7 = Pose Driver - Source: clavicle_l!(SourcePose:Pose Link Structure=Blend Poses by bool(BlendPose_0=Use cached pose 'noRIgids_cache'(), BlendPose_1=Component To Local(ComponentPose=@n4), bActiveValue=OR Boolean(A=@n6, B=@n5)) {blendType=Linear}) -> (Pose:Pose Link Structure) {
    is=title AnimGraphNode_PoseDriver
    poseAsset=&/F_MED_SparkleRemix_PSA_PoseAsset
    evalSpaceBone {boneName=spine_05}
    onlyDriveBones [
     {boneName=dyn_shoulderPadA_l}
     {boneName=dyn_shoulderPadB_l}
     {boneName=dyn_shoulderPadC_l}
     {boneName=dyn_shoulderPadD_l}
     {boneName=dyn_shoulderPadE_l}
    ]
    poseTargets [
     {
      bApplyCustomCurve=false
      bIsHidden=false
      distanceMethod=DefaultMethod
      drivenName=Ref
      functionType=DefaultFunction
      targetScale=0.9434
      boneTransforms [
       {targetRotation=<rot -85.1913 -138.5623 143.0617>, targetTranslation=<5.4831 -0.1053 -3.2946>}
       {targetRotation=<rot -48.2466 -175.6266 -178.9752>, targetTranslation=<4.5519 -1.8922 -13.037>}
      ]
      customCurve {defaultValue=3.4028234663852886e+38, postInfinityExtrap=RCCE_Constant, preInfinityExtrap=RCCE_Constant, keys []}
      targetRotation=<rot 0 0 0>
     }
     {
      bApplyCustomCurve=false
      bIsHidden=false
      distanceMethod=DefaultMethod
      drivenName=PistolFor
      functionType=DefaultFunction
      targetScale=0.6487
      boneTransforms [
       {targetRotation=<rot -79.5583 65.3944 -62.1168>, targetTranslation=<5.4831 -0.1053 -3.2946>}
       {targetRotation=<rot -25.1943 130.2148 -173.8125>, targetTranslation=<5.8959 0.3043 -13.2262>}
      ]
      customCurve {defaultValue=3.4028234663852886e+38, postInfinityExtrap=RCCE_Constant, preInfinityExtrap=RCCE_Constant, keys []}
      targetRotation=<rot 0 0 0>
     }
     {
      bApplyCustomCurve=false
      bIsHidden=false
      distanceMethod=DefaultMethod
      drivenName=PistolUp
      functionType=DefaultFunction
      targetScale=1.5
      boneTransforms [
       {targetRotation=<rot -49.7632 16.2825 -5.5507>, targetTranslation=<5.4831 -0.1053 -3.2946>}
       {targetRotation=<rot -25.6894 66.7378 -144.862>, targetTranslation=<11.5424 0.4475 -11.1657>}
      ]
      customCurve {defaultValue=3.4028234663852886e+38, postInfinityExtrap=RCCE_Constant, preInfinityExtrap=RCCE_Constant, keys []}
      targetRotation=<rot 0 0 0>
     }
     {
      bApplyCustomCurve=false
      bIsHidden=false
      distanceMethod=DefaultMethod
      drivenName=RifleFor
      functionType=DefaultFunction
      targetScale=0.1
      boneTransforms [
       {targetRotation=<rot -82.2703 53.7584 -49.741>, targetTranslation=<5.4831 -0.1053 -3.2946>}
       {targetRotation=<rot -42.254 143.8637 163.1364>, targetTranslation=<5.9495 -0.2657 -13.231>}
      ]
      customCurve {defaultValue=3.4028234663852886e+38, postInfinityExtrap=RCCE_Constant, preInfinityExtrap=RCCE_Constant, keys []}
      targetRotation=<rot 0 0 0>
     }
     {
      bApplyCustomCurve=false
      bIsHidden=false
      distanceMethod=DefaultMethod
      drivenName=RifleUp
      functionType=DefaultFunction
      targetScale=0.4
      boneTransforms [
       {targetRotation=<rot -47.5989 14.683 -0.3775>, targetTranslation=<5.4831 -0.1053 -3.2946>}
       {targetRotation=<rot -60.6828 51.5686 -105.0432>, targetTranslation=<11.9265 0.3313 -10.8621>}
      ]
      customCurve {defaultValue=3.4028234663852886e+38, postInfinityExtrap=RCCE_Constant, preInfinityExtrap=RCCE_Constant, keys []}
      targetRotation=<rot 0 0 0>
     }
     {
      bApplyCustomCurve=false
      bIsHidden=false
      distanceMethod=DefaultMethod
      drivenName=ShotFor
      functionType=DefaultFunction
      targetScale=0
      boneTransforms [
       {targetRotation=<rot -82.9391 61.7287 -60.781>, targetTranslation=<5.4831 -0.1053 -3.2946>}
       {targetRotation=<rot -40.7858 146.4693 170.6522>, targetTranslation=<5.6717 -0.249 -13.2404>}
      ]
      customCurve {defaultValue=3.4028234663852886e+38, postInfinityExtrap=RCCE_Constant, preInfinityExtrap=RCCE_Constant, keys []}
      targetRotation=<rot 0 0 0>
     }
     {
      bApplyCustomCurve=false
      bIsHidden=false
      distanceMethod=DefaultMethod
      drivenName=ShotUp
      functionType=DefaultFunction
      targetScale=0.3671
      boneTransforms [
       {targetRotation=<rot -46.6794 15.7848 -6.0623>, targetTranslation=<5.4831 -0.1053 -3.2946>}
       {targetRotation=<rot -50.6183 41.0661 -97.5768>, targetTranslation=<11.9376 0.5102 -10.8402>}
      ]
      customCurve {defaultValue=3.4028234663852886e+38, postInfinityExtrap=RCCE_Constant, preInfinityExtrap=RCCE_Constant, keys []}
      targetRotation=<rot 0 0 0>
     }
    ]
    rBFParams {radius=15.9484}
    sourceBones [{boneName=clavicle_l}, {boneName=upperarm_l}]
   }
   n8 = Pose Driver - Source: clavicle_r!(SourcePose:Pose Link Structure=@n7) -> (Pose:Pose Link Structure) {
    is=title AnimGraphNode_PoseDriver
    poseAsset=&F_MED_SparkleRemix_PSA_PoseAsset
    evalSpaceBone {boneName=spine_05}
    onlyDriveBones [
     {boneName=dyn_shoulderPadA_r}
     {boneName=dyn_shoulderPadB_r}
     {boneName=dyn_shoulderPadC_r}
     {boneName=dyn_shoulderPadD_r}
     {boneName=dyn_shoulderPadE_r}
    ]
    poseTargets [
     {
      bApplyCustomCurve=false
      bIsHidden=false
      distanceMethod=DefaultMethod
      drivenName=Ref
      functionType=DefaultFunction
      targetScale=1
      boneTransforms [
       {targetRotation=<rot -85.1913 41.4375 143.0617>, targetTranslation=<5.4831 -0.1053 3.2946>}
       {targetRotation=<rot -48.2466 4.3734 -178.9752>, targetTranslation=<4.552 -1.8922 13.037>}
      ]
      customCurve {defaultValue=3.4028234663852886e+38, postInfinityExtrap=RCCE_Constant, preInfinityExtrap=RCCE_Constant, keys []}
      targetRotation=<rot 0 0 0>
     }
     {
      bApplyCustomCurve=false
      bIsHidden=false
      distanceMethod=DefaultMethod
      drivenName=PistolFor
      functionType=DefaultFunction
      targetScale=0.725
      boneTransforms [
       {targetRotation=<rot -68.9942 -127.628 -53.9142>, targetTranslation=<5.4831 -0.1053 3.2946>}
       {targetRotation=<rot 23.1305 -78.1139 -124.5826>, targetTranslation=<7.2489 1.5578 12.943>}
      ]
      customCurve {defaultValue=3.4028234663852886e+38, postInfinityExtrap=RCCE_Constant, preInfinityExtrap=RCCE_Constant, keys []}
      targetRotation=<rot 0 0 0>
     }
     {
      bApplyCustomCurve=false
      bIsHidden=false
      distanceMethod=DefaultMethod
      drivenName=PistolUp
      functionType=DefaultFunction
      targetScale=0.7302
      boneTransforms [
       {targetRotation=<rot -68.9927 -127.6306 -53.9082>, targetTranslation=<5.4831 -0.1053 3.2946>}
       {targetRotation=<rot 23.5002 -120.7419 -169.1149>, targetTranslation=<7.2493 1.5579 12.9429>}
      ]
      customCurve {defaultValue=3.4028234663852886e+38, postInfinityExtrap=RCCE_Constant, preInfinityExtrap=RCCE_Constant, keys []}
      targetRotation=<rot 0 0 0>
     }
     {
      bApplyCustomCurve=false
      bIsHidden=false
      distanceMethod=DefaultMethod
      drivenName=RifleFor
      functionType=DefaultFunction
      targetScale=0.0543
      boneTransforms [
       {targetRotation=<rot -75.121 -133.534 -46.1447>, targetTranslation=<5.4831 -0.1053 3.2946>}
       {targetRotation=<rot 1.8346 -40.281 -117.1515>, targetTranslation=<6.8512 0.5436 13.1273>}
      ]
      customCurve {defaultValue=3.4028234663852886e+38, postInfinityExtrap=RCCE_Constant, preInfinityExtrap=RCCE_Constant, keys []}
      targetRotation=<rot 0 0 0>
     }
     {
      bApplyCustomCurve=false
      bIsHidden=false
      distanceMethod=DefaultMethod
      drivenName=RifleUp
      functionType=DefaultFunction
      targetScale=0.38
      boneTransforms [
       {targetRotation=<rot -74.8462 -132.6499 -47.0321>, targetTranslation=<5.4831 -0.1053 3.2946>}
       {targetRotation=<rot 14.0717 -49.1056 -123.8755>, targetTranslation=<6.8543 0.6056 13.1226>}
      ]
      customCurve {defaultValue=3.4028234663852886e+38, postInfinityExtrap=RCCE_Constant, preInfinityExtrap=RCCE_Constant, keys []}
      targetRotation=<rot 0 0 0>
     }
     {
      bApplyCustomCurve=false
      bIsHidden=false
      distanceMethod=DefaultMethod
      drivenName=ShotFor
      functionType=DefaultFunction
      targetScale=0.2673
      boneTransforms [
       {targetRotation=<rot -74.3679 -136.1982 -43.567>, targetTranslation=<5.4831 -0.1053 3.2946>}
       {targetRotation=<rot -11.4857 -35.2179 -118.611>, targetTranslation=<7.0266 0.5485 13.101>}
      ]
      customCurve {defaultValue=3.4028234663852886e+38, postInfinityExtrap=RCCE_Constant, preInfinityExtrap=RCCE_Constant, keys []}
      targetRotation=<rot 0 0 0>
     }
     {
      bApplyCustomCurve=false
      bIsHidden=false
      distanceMethod=DefaultMethod
      drivenName=ShotUp
      functionType=DefaultFunction
      targetScale=0
      boneTransforms [
       {targetRotation=<rot -75.387 -134.2757 -45.4136>, targetTranslation=<5.4831 -0.1053 3.2946>}
       {targetRotation=<rot 2.987 -42.9259 -127.374>, targetTranslation=<6.843 0.4877 13.132>}
      ]
      customCurve {defaultValue=3.4028234663852886e+38, postInfinityExtrap=RCCE_Constant, preInfinityExtrap=RCCE_Constant, keys []}
      targetRotation=<rot 0 0 0>
     }
    ]
    rBFParams {radius=26.7071}
    sourceBones [{boneName=clavicle_r}, {boneName=upperarm_r}]
   }
   Output Pose(Result=@n8) {layerGroup=None}
   n9 = "Constraint - dyn_shoulderPadA_Base_r (PPP)"!(
    ComponentPose=Local To Component(LocalPose=Copy Pose From Mesh() {bUseAttachedParent=true})
    Alpha=1
    ConstraintWeights_0=0.7
    ConstraintWeights_1:float=0.2
    ConstraintWeights_2:float=0.1
   ) -> (Pose) {
    alphaBoolBlend {blendOption=Linear}
    boneToModify {boneName=dyn_shoulderPadA_Base_r}
    constraintSetup [{targetBone {boneName=clavicle_r}}, {targetBone {boneName=upperarm_r}}, {targetBone {boneName=spine_05}}]
    constraintWeights [1,1,1]
   }
   n10 = "Constraint - dyn_shoulderPadB_Base_r (PPP)"!(ComponentPose=@n9, Alpha=1, ConstraintWeights_0=0.5, ConstraintWeights_1:float=0.4, ConstraintWeights_2:float=0.1) -> (Pose) {
    alphaBoolBlend {blendOption=Linear}
    boneToModify {boneName=dyn_shoulderPadB_Base_r}
    constraintSetup [{targetBone {boneName=clavicle_r}}, {targetBone {boneName=upperarm_r}}, {targetBone {boneName=spine_05}}]
    constraintWeights [1,1,1]
   }
   n11 = "Constraint - dyn_shoulderPadC_Base_r (PPP)"!(ComponentPose=@n10, Alpha=1, ConstraintWeights_0=0.8, ConstraintWeights_1:float=0.1, ConstraintWeights_2:float=0.1) -> (Pose) {
    alphaBoolBlend {blendOption=Linear}
    boneToModify {boneName=dyn_shoulderPadC_Base_r}
    constraintSetup [{targetBone {boneName=clavicle_r}}, {targetBone {boneName=upperarm_r}}, {targetBone {boneName=spine_05}}]
    constraintWeights [1,1,1]
   }
   n12 = "Constraint - dyn_shoulderPadD_Base_r (PPP)"!(ComponentPose=@n11, Alpha=1, ConstraintWeights_0=1, ConstraintWeights_1:float=0, ConstraintWeights_2:float=0) -> (Pose) {
    alphaBoolBlend {blendOption=Linear}
    boneToModify {boneName=dyn_shoulderPadD_Base_r}
    constraintSetup [{targetBone {boneName=clavicle_r}}, {targetBone {boneName=upperarm_r}}, {targetBone {boneName=spine_05}}]
    constraintWeights [1,1,1]
   }
   n13 = "Constraint - dyn_shoulderPadE_Base_r (PPP)"!(ComponentPose=@n12, Alpha=1, ConstraintWeights_0=1, ConstraintWeights_1:float=0, ConstraintWeights_2:float=0) -> (Pose) {
    alphaBoolBlend {blendOption=Linear}
    boneToModify {boneName=dyn_shoulderPadE_Base_r}
    constraintSetup [{targetBone {boneName=clavicle_r}}, {targetBone {boneName=upperarm_r}}, {targetBone {boneName=spine_05}}]
    constraintWeights [1,1,1]
   }
   n14 = "Constraint - dyn_shoulderPadA_Base_l (PPP)"!(ComponentPose=@n13, Alpha=1, ConstraintWeights_0=0.7, ConstraintWeights_1:float=0.2, ConstraintWeights_2:float=0.1) -> (Pose) {
    alphaBoolBlend {blendOption=Linear}
    boneToModify {boneName=dyn_shoulderPadA_Base_l}
    constraintSetup [{targetBone {boneName=clavicle_l}}, {targetBone {boneName=upperarm_l}}, {targetBone {boneName=spine_05}}]
    constraintWeights [1,1,1]
   }
   n15 = "Constraint - dyn_shoulderPadB_Base_l (PPP)"!(ComponentPose=@n14, Alpha=1, ConstraintWeights_0=0.5, ConstraintWeights_1:float=0.4, ConstraintWeights_2:float=0.1) -> (Pose) {
    alphaBoolBlend {blendOption=Linear}
    boneToModify {boneName=dyn_shoulderPadB_Base_l}
    constraintSetup [{targetBone {boneName=clavicle_l}}, {targetBone {boneName=upperarm_l}}, {targetBone {boneName=spine_05}}]
    constraintWeights [1,1,1]
   }
   n16 = "Constraint - dyn_shoulderPadC_Base_l (PPP)"!(ComponentPose=@n15, Alpha=1, ConstraintWeights_0=0.8, ConstraintWeights_1:float=0.1, ConstraintWeights_2:float=0.1) -> (Pose) {
    alphaBoolBlend {blendOption=Linear}
    boneToModify {boneName=dyn_shoulderPadC_Base_l}
    constraintSetup [{targetBone {boneName=clavicle_l}}, {targetBone {boneName=upperarm_l}}, {targetBone {boneName=spine_05}}]
    constraintWeights [1,1,1]
   }
   n17 = "Constraint - dyn_shoulderPadD_Base_l (PPP)"!(ComponentPose=@n16, Alpha=1, ConstraintWeights_0=1, ConstraintWeights_1:float=0, ConstraintWeights_2:float=0) -> (Pose) {
    alphaBoolBlend {blendOption=Linear}
    boneToModify {boneName=dyn_shoulderPadD_Base_l}
    constraintSetup [{targetBone {boneName=clavicle_l}}, {targetBone {boneName=upperarm_l}}, {targetBone {boneName=spine_05}}]
    constraintWeights [1,1,1]
   }
   n18 = "Constraint - dyn_shoulderPadE_Base_l (PPP)"!(ComponentPose=@n17, Alpha=1, ConstraintWeights_0=1, ConstraintWeights_1:float=0, ConstraintWeights_2:float=0) -> (Pose) {
    alphaBoolBlend {blendOption=Linear}
    boneToModify {boneName=dyn_shoulderPadE_Base_l}
    constraintSetup [{targetBone {boneName=clavicle_l}}, {targetBone {boneName=upperarm_l}}, {targetBone {boneName=spine_05}}]
    constraintWeights [1,1,1]
   }
   n19 = Get armSwingDelta_r() -> (armSwingDelta_r:double)
   n20 = "Transform (Modify) Bone - Bone: dyn_shoulderPadA_Base_r"(ComponentPose=@n18, Alpha=@n19, Translation="0,-1,0", Rotation="0,-5,0", Scale="1,1,1") {
    rotationMode=BMM_Additive
    rotationSpace=BCS_BoneSpace
    translationMode=BMM_Additive
    translationSpace=BCS_BoneSpace
    alphaBoolBlend {blendOption=Linear}
    boneToModify {boneName=dyn_shoulderPadA_Base_r}
    rotation {yaw=-5}
    translation {y=-1}
   }
   n21 = "Transform (Modify) Bone - Bone: dyn_shoulderPadB_Base_r"(ComponentPose=@n20, Alpha=@n19, Translation="0,-2,0", Rotation="0,-22,0", Scale="1,1,1") {
    rotationMode=BMM_Additive
    rotationSpace=BCS_BoneSpace
    translationMode=BMM_Additive
    translationSpace=BCS_BoneSpace
    alphaBoolBlend {blendOption=Linear}
    boneToModify {boneName=dyn_shoulderPadB_Base_r}
    rotation {yaw=-22}
    translation {y=-2}
   }
   n22 = "Transform (Modify) Bone - Bone: dyn_shoulderPadC_Base_r"(ComponentPose=@n21, Alpha=@n19, Translation="0,-2,0", Rotation="0,-10,0", Scale="1,1,1") {
    rotationMode=BMM_Additive
    rotationSpace=BCS_BoneSpace
    translationMode=BMM_Additive
    translationSpace=BCS_BoneSpace
    alphaBoolBlend {blendOption=Linear}
    boneToModify {boneName=dyn_shoulderPadC_Base_r}
    rotation {yaw=-10}
    translation {y=-2}
   }
   n23 = Get armSwingDelta_l() -> (armSwingDelta_l:double)
   n24 = "Transform (Modify) Bone - Bone: dyn_shoulderPadA_Base_l"(ComponentPose=@n22, Alpha=@n23, Translation="0,0,0", Rotation="0,-5,0", Scale="1,1,1") {
    rotationMode=BMM_Additive
    rotationSpace=BCS_BoneSpace
    translationMode=BMM_Additive
    translationSpace=BCS_BoneSpace
    alphaBoolBlend {blendOption=Linear}
    boneToModify {boneName=dyn_shoulderPadA_Base_l}
    rotation {yaw=-5}
   }
   n25 = "Transform (Modify) Bone - Bone: dyn_shoulderPadB_Base_l"(ComponentPose=@n24, Alpha=@n23, Translation="0,0,0", Rotation="0,-22,0", Scale="1,1,1") {
    rotationMode=BMM_Additive
    rotationSpace=BCS_BoneSpace
    translationMode=BMM_Additive
    translationSpace=BCS_BoneSpace
    alphaBoolBlend {blendOption=Linear}
    boneToModify {boneName=dyn_shoulderPadB_Base_l}
    rotation {yaw=-22}
   }
   n26 = "Transform (Modify) Bone - Bone: dyn_shoulderPadC_Base_l"(ComponentPose=@n25, Alpha=@n23, Translation="0,0,0", Rotation="0,-10,0", Scale="1,1,1") {
    rotationMode=BMM_Additive
    rotationSpace=BCS_BoneSpace
    translationMode=BMM_Additive
    translationSpace=BCS_BoneSpace
    alphaBoolBlend {blendOption=Linear}
    boneToModify {boneName=dyn_shoulderPadC_Base_l}
    rotation {yaw=-10}
   }
   n27 = "Transform (Modify) Bone - Bone: dyn_shoulderPadC_Base_r"(ComponentPose=@n26, Alpha=Get clavSwingDelta_r() -> (clavSwingDelta_r:double), Translation="0,0,0", Rotation="0,-30,0", Scale="1,1,1.5") {
    rotationMode=BMM_Additive
    rotationSpace=BCS_BoneSpace
    scaleMode=BMM_Replace
    scaleSpace=BCS_BoneSpace
    translationMode=BMM_Additive
    translationSpace=BCS_BoneSpace
    alphaBoolBlend {blendOption=Linear}
    boneToModify {boneName=dyn_shoulderPadC_Base_r}
    rotation {yaw=-30}
    scale {z=1.5}
   }
   n28 = "Transform (Modify) Bone - Bone: dyn_shoulderPadC_Base_l"(ComponentPose=@n27, Alpha=Get clavSwingDelta_l() -> (clavSwingDelta_l:double), Translation="0,0,0", Rotation="0,-30,0", Scale="1,1,1.5") {
    rotationMode=BMM_Additive
    rotationSpace=BCS_BoneSpace
    translationMode=BMM_Additive
    translationSpace=BCS_BoneSpace
    alphaBoolBlend {blendOption=Linear}
    boneToModify {boneName=dyn_shoulderPadC_Base_l}
    rotation {yaw=-30}
    scale {z=1.5}
   }
   n29 = Get bIsSkydiveFloating()
   n30 = "Transform (Modify) Bone - Bone: dyn_shoulderPadC_Base_r"!(ComponentPose=@n28, bAlphaBoolEnabled:bool=@n29, Translation="0,0,0", Rotation="0,-15,0", Scale="1,1,1.5") -> (Pose) {
    alphaInputType=Bool
    rotationMode=BMM_Additive
    rotationSpace=BCS_BoneSpace
    scaleMode=BMM_Replace
    scaleSpace=BCS_BoneSpace
    translationMode=BMM_Additive
    translationSpace=BCS_BoneSpace
    alphaBoolBlend {blendInTime=1, blendOption=Linear, blendOutTime=1}
    boneToModify {boneName=dyn_shoulderPadC_Base_r}
    rotation {yaw=-15}
    scale {z=1.5}
   }
   n31 = "Transform (Modify) Bone - Bone: dyn_shoulderPadC_Base_l"!(ComponentPose=@n30, bAlphaBoolEnabled:bool=@n29, Translation="0,0,0", Rotation="0,-30,0", Scale="1,1,1.5") -> (Pose) {
    alphaInputType=Bool
    rotationMode=BMM_Additive
    rotationSpace=BCS_BoneSpace
    translationMode=BMM_Additive
    translationSpace=BCS_BoneSpace
    alphaBoolBlend {blendInTime=1, blendOption=Linear, blendOutTime=1}
    boneToModify {boneName=dyn_shoulderPadC_Base_l}
    rotation {yaw=-30}
    scale {z=1.5}
   }
   n32 = Get bIsSkydiveDiveMode()
   n33 = "Transform (Modify) Bone - Bone: dyn_shoulderPadD_Base_r"!(ComponentPose=@n31, bAlphaBoolEnabled:bool=@n32, Translation="0,-5,0", Rotation="0,20,0", Scale="1,1,1.5") -> (Pose) {
    alphaInputType=Bool
    rotationMode=BMM_Additive
    rotationSpace=BCS_BoneSpace
    scaleMode=BMM_Replace
    scaleSpace=BCS_BoneSpace
    translationMode=BMM_Additive
    translationSpace=BCS_BoneSpace
    alphaBoolBlend {blendInTime=0.1, blendOption=Linear, blendOutTime=0.1}
    boneToModify {boneName=dyn_shoulderPadD_Base_r}
    rotation {yaw=20}
    scale {z=1.5}
    translation {y=-5}
   }
   n34 = "Transform (Modify) Bone - Bone: dyn_shoulderPadE_Base_r"!(ComponentPose=@n33, bAlphaBoolEnabled:bool=@n32, Translation="0,-5,0", Rotation="0,20,0", Scale="1,1,1.5") -> (Pose) {
    alphaInputType=Bool
    rotationMode=BMM_Additive
    rotationSpace=BCS_BoneSpace
    scaleMode=BMM_Replace
    scaleSpace=BCS_BoneSpace
    translationMode=BMM_Additive
    translationSpace=BCS_BoneSpace
    alphaBoolBlend {blendInTime=0.1, blendOption=Linear, blendOutTime=0.1}
    boneToModify {boneName=dyn_shoulderPadE_Base_r}
    rotation {yaw=20}
    scale {z=1.5}
    translation {y=-5}
   }
   n35 = "Transform (Modify) Bone - Bone: dyn_shoulderPadD_Base_l"!(ComponentPose=@n34, bAlphaBoolEnabled:bool=@n32, Translation="0,5,0", Rotation="0,20,0", Scale="1,1,1.5") -> (Pose) {
    alphaInputType=Bool
    rotationMode=BMM_Additive
    rotationSpace=BCS_BoneSpace
    translationMode=BMM_Additive
    translationSpace=BCS_BoneSpace
    alphaBoolBlend {blendInTime=0.1, blendOption=Linear, blendOutTime=0.1}
    boneToModify {boneName=dyn_shoulderPadD_Base_l}
    rotation {yaw=20}
    scale {z=1.5}
    translation {y=5}
   }
   n36 = "Transform (Modify) Bone - Bone: dyn_shoulderPadE_Base_l"!(ComponentPose=@n35, bAlphaBoolEnabled:bool=@n32, Translation="0,5,0", Rotation="0,20,0", Scale="1,1,1.5") -> (Pose) {
    alphaInputType=Bool
    rotationMode=BMM_Additive
    rotationSpace=BCS_BoneSpace
    translationMode=BMM_Additive
    translationSpace=BCS_BoneSpace
    alphaBoolBlend {blendInTime=0.1, blendOption=Linear, blendOutTime=0.1}
    boneToModify {boneName=dyn_shoulderPadE_Base_l}
    rotation {yaw=20}
    scale {z=1.5}
    translation {y=5}
   }
   Save cached pose 'preCrouch'(Pose=Component To Local(ComponentPose=@n36)) {cacheName=preCrouch}
   n37 = Get armSwingDelta_l()
   n38 = Get armSwingDelta_r()
   n39 = "Transform (Modify) Bone - Bone: dyn_shoulderPadA_Base_r"(ComponentPose=Local To Component(LocalPose=Use cached pose 'preCrouch'()), Alpha=@n38, Translation="0,-1,0", Rotation="0,0,0", Scale="1,1,1") {
    rotationMode=BMM_Additive
    rotationSpace=BCS_BoneSpace
    translationMode=BMM_Additive
    translationSpace=BCS_BoneSpace
    alphaBoolBlend {blendOption=Linear}
    boneToModify {boneName=dyn_shoulderPadA_Base_r}
    translation {y=-1}
   }
   n40 = "Transform (Modify) Bone - Bone: dyn_shoulderPadB_Base_r"(ComponentPose=@n39, Alpha=@n38, Translation="-5,-5,0", Rotation="0,-20,0", Scale="1,1,1") {
    rotationMode=BMM_Additive
    rotationSpace=BCS_BoneSpace
    translationMode=BMM_Additive
    translationSpace=BCS_BoneSpace
    alphaBoolBlend {blendOption=Linear}
    boneToModify {boneName=dyn_shoulderPadB_Base_r}
    rotation {yaw=-20}
    translation=<-5 -5>
   }
   n41 = "Transform (Modify) Bone - Bone: dyn_shoulderPadC_r"(ComponentPose=@n40, Alpha=@n38, Translation="-5,-5,0", Rotation="0,-20,0", Scale="1,1,1") {
    rotationMode=BMM_Additive
    rotationSpace=BCS_BoneSpace
    translationMode=BMM_Additive
    translationSpace=BCS_BoneSpace
    alphaBoolBlend {blendOption=Linear}
    boneToModify {boneName=dyn_shoulderPadC_r}
    rotation {yaw=-20}
    translation=<-5 -5>
   }
   n42 = "Transform (Modify) Bone - Bone: dyn_shoulderPadD_Base_r"!(ComponentPose=@n41, Alpha:float=@n38, Translation="0,-3,0", Rotation="0,-20,0", Scale="1,1,1") -> (Pose) {
    rotationMode=BMM_Additive
    rotationSpace=BCS_BoneSpace
    translationMode=BMM_Additive
    translationSpace=BCS_BoneSpace
    alphaBoolBlend {blendOption=Linear}
    boneToModify {boneName=dyn_shoulderPadD_Base_r}
    rotation {yaw=-20}
    translation {y=-3}
   }
   n43 = "Transform (Modify) Bone - Bone: dyn_shoulderPadE_Base_r"!(ComponentPose=@n42, Alpha:float=@n38, Translation="0,-3,0", Rotation="0.000,-22,0.000", Scale="1,1,1") -> (Pose) {
    rotationMode=BMM_Additive
    rotationSpace=BCS_BoneSpace
    translationMode=BMM_Additive
    translationSpace=BCS_BoneSpace
    alphaBoolBlend {blendOption=Linear}
    boneToModify {boneName=dyn_shoulderPadE_Base_r}
    rotation {yaw=-22}
    translation {y=-3}
   }
   n44 = "Transform (Modify) Bone - Bone: dyn_shoulderPadA_Base_l"(ComponentPose=@n43, Alpha=@n37, Translation="0,2,0", Rotation="0,-10,0", Scale="1,1,1") {
    rotationMode=BMM_Additive
    rotationSpace=BCS_BoneSpace
    translationMode=BMM_Additive
    translationSpace=BCS_BoneSpace
    alphaBoolBlend {blendOption=Linear}
    boneToModify {boneName=dyn_shoulderPadA_Base_l}
    rotation {yaw=-10}
    translation {y=2}
   }
   n45 = "Transform (Modify) Bone - Bone: dyn_shoulderPadB_Base_l"(ComponentPose=@n44, Alpha=@n37, Translation="0,5,0", Rotation="0,-10,0", Scale="1,1,1") {
    rotationMode=BMM_Additive
    rotationSpace=BCS_BoneSpace
    translationMode=BMM_Additive
    translationSpace=BCS_BoneSpace
    alphaBoolBlend {blendOption=Linear}
    boneToModify {boneName=dyn_shoulderPadB_Base_l}
    rotation {yaw=-10}
    translation {y=5}
   }
   n46 = "Transform (Modify) Bone - Bone: dyn_shoulderPadC_Base_l"!(ComponentPose=@n45, Alpha:float=@n37, Translation="0,5,0", Rotation="0,-10,0", Scale="1,1,1") -> (Pose) {
    rotationMode=BMM_Additive
    rotationSpace=BCS_BoneSpace
    translationMode=BMM_Additive
    translationSpace=BCS_BoneSpace
    alphaBoolBlend {blendOption=Linear}
    boneToModify {boneName=dyn_shoulderPadC_Base_l}
    rotation {yaw=-10}
    translation {y=5}
   }
   n47 = "Transform (Modify) Bone - Bone: dyn_shoulderPadD_Base_l"!(ComponentPose=@n46, Alpha:float=@n37, Translation="0,3,0", Rotation="0,-20,0", Scale="1,1,1") -> (Pose) {
    rotationMode=BMM_Additive
    rotationSpace=BCS_BoneSpace
    translationMode=BMM_Additive
    translationSpace=BCS_BoneSpace
    alphaBoolBlend {blendOption=Linear}
    boneToModify {boneName=dyn_shoulderPadD_Base_l}
    rotation {yaw=-20}
    translation {y=3}
   }
   n48 = "Transform (Modify) Bone - Bone: dyn_shoulderPadE_Base_l"!(ComponentPose=@n47, Alpha:float=@n37, Translation="0,3,0", Rotation="0.000,-22,0.000", Scale="1,1,1") -> (Pose) {
    rotationMode=BMM_Additive
    rotationSpace=BCS_BoneSpace
    translationMode=BMM_Additive
    translationSpace=BCS_BoneSpace
    alphaBoolBlend {blendOption=Linear}
    boneToModify {boneName=dyn_shoulderPadE_Base_l}
    rotation {yaw=-22}
    translation {y=3}
   }
   n49 = Component To Local(ComponentPose=@n48)
   n50 = Use cached pose 'preCrouch'()
   n51 = Get crouchAlpha()
   n52 = "Constraint - dyn_backCrease_Base_r (PP)"!(
    ComponentPose=Local To Component(LocalPose=Blend Poses by bool(BlendPose_0=@n49, BlendPose_1=@n50, BlendTime_0=0.15, BlendTime_1=0.15, bActiveValue=@n51) {blendType=Linear})
    Alpha=1
    ConstraintWeights_0=0.5
    ConstraintWeights_1:float=0.5
   ) -> (Pose) {
    alphaBoolBlend {blendOption=Linear}
    boneToModify {boneName=dyn_backCrease_Base_r}
    constraintSetup [{targetBone {boneName=dyn_skirtFlapBack_01_r}}, {targetBone {boneName=dyn_skirtFlapSide_01_r}}]
    constraintWeights [1,1]
   }
   n53 = "Constraint - dyn_backCrease_Base_l (PP)"!(ComponentPose=@n52, Alpha=1, ConstraintWeights_0=0.5, ConstraintWeights_1:float=0.5) -> (Pose) {
    alphaBoolBlend {blendOption=Linear}
    boneToModify {boneName=dyn_backCrease_Base_l}
    constraintSetup [{targetBone {boneName=dyn_skirtFlapBack_01_l}}, {targetBone {boneName=dyn_skirtFlapSide_01_l}}]
    constraintWeights [1,1]
   }
   Save cached pose 'preRigids'(Pose=Component To Local(ComponentPose=@n53)) {cacheName=preRigids}
   n54 = "Constraint - dyn_skirtFlap_02_r (P)"(ComponentPose=Local To Component(LocalPose=Use cached pose 'preRigids'()), Alpha=1, ConstraintWeights_0=1) {
    alphaBoolBlend {blendOption=Linear}
    boneToModify {boneName=dyn_skirtFlap_02_r}
    constraintSetup [{targetBone {boneName=thigh_r}}]
    constraintWeights [1]
   }
   n55 = "Constraint - dyn_skirtFlapSide_01_r (PP)"!(ComponentPose=@n54, Alpha=1, ConstraintWeights_0=0.5, ConstraintWeights_1:float=0.5) -> (Pose) {
    alphaBoolBlend {blendOption=Linear}
    boneToModify {boneName=dyn_skirtFlapSide_01_r}
    constraintSetup [{targetBone {boneName=thigh_r}}, {targetBone {boneName=pelvis}}]
    constraintWeights [1,1]
   }
   n56 = "Constraint - dyn_skirtFlapBack_01_r (PP)"!(ComponentPose=@n55, Alpha=1, ConstraintWeights_0=0.2, ConstraintWeights_1:float=0.8) -> (Pose) {
    alphaBoolBlend {blendOption=Linear}
    boneToModify {boneName=dyn_skirtFlapBack_01_r}
    constraintSetup [{targetBone {boneName=dyn_skirtFlapSide_01_r}}, {targetBone {boneName=pelvis}}]
    constraintWeights [1,1]
   }
   n57 = "Constraint - dyn_loincloth_02_r (PPT)"!(ComponentPose=@n56, Alpha=1, ConstraintWeights_0=0.3, ConstraintWeights_1:float=0.7, ConstraintWeights_2:float=0) -> (Pose) {
    alphaBoolBlend {blendOption=Linear}
    boneToModify {boneName=dyn_loincloth_02_r}
    constraintSetup [{targetBone {boneName=thigh_r}}, {targetBone {boneName=pelvis}}, {transformType=Translation, targetBone {boneName=None}}]
    constraintWeights [1,1,1]
   }
   n58 = "Constraint - dyn_skirtFlap_02_l (P)"(ComponentPose=@n57, Alpha=1, ConstraintWeights_0=1) {
    alphaBoolBlend {blendOption=Linear}
    boneToModify {boneName=dyn_skirtFlap_02_l}
    constraintSetup [{targetBone {boneName=thigh_l}}]
    constraintWeights [1]
   }
   n59 = "Constraint - dyn_skirtFlapSide_02_l (PP)"!(ComponentPose=@n58, Alpha=1, ConstraintWeights_0=0.5, ConstraintWeights_1:float=0.5) -> (Pose) {
    alphaBoolBlend {blendOption=Linear}
    boneToModify {boneName=dyn_skirtFlapSide_02_l}
    constraintSetup [{targetBone {boneName=thigh_l}}, {targetBone {boneName=pelvis}}]
    constraintWeights [1,1]
   }
   n60 = "Constraint - dyn_skirtFlapBack_01_l (PP)"!(ComponentPose=@n59, Alpha=1, ConstraintWeights_0=0.2, ConstraintWeights_1:float=0.8) -> (Pose) {
    alphaBoolBlend {blendOption=Linear}
    boneToModify {boneName=dyn_skirtFlapBack_01_l}
    constraintSetup [{targetBone {boneName=dyn_skirtFlapSide_01_l}}, {targetBone {boneName=pelvis}}]
    constraintWeights [1,1]
   }
   n61 = "Constraint - dyn_loincloth_02_l (PPT)"!(ComponentPose=@n60, Alpha=1, ConstraintWeights_0=0.3, ConstraintWeights_1:float=0.7, ConstraintWeights_2:float=0) -> (Pose) {
    alphaBoolBlend {blendOption=Linear}
    boneToModify {boneName=dyn_loincloth_02_l}
    constraintSetup [{targetBone {boneName=thigh_l}}, {targetBone {boneName=pelvis}}, {transformType=Translation, targetBone {boneName=None}}]
    constraintWeights [1,1,1]
   }
   Save cached pose 'noRIgids_cache'(Pose=Component To Local(ComponentPose=@n61)) {cacheName=noRIgids_cache}
  ]}
 EventGraph {nodes [
   Event BlueprintUpdateAnimation()
   n1 = GetOwnerSkeletalMeshComponent()
   n2 = BreakRotator(InRot=BreakTransform(InTransform=GetDeltaTransformFromRefPose(self=@n1, BoneName=upperarm_r, BaseName=clavicle_r)).Rotation)
   > Set armSwingDelta_r(armSwingDelta_r:double=MapRangeClamped(Value="Absolute (Float)"(A=@n2.Yaw), InRangeA=45, InRangeB=150)) -> (Output_Get:double)
   n3 = BreakRotator(InRot=BreakTransform(InTransform=GetDeltaTransformFromRefPose(self=@n1, BoneName=upperarm_l, BaseName=clavicle_l)).Rotation)
   > Set armSwingDelta_l(armSwingDelta_l:double=MapRangeClamped(Value="Absolute (Float)"(A=@n3.Yaw), InRangeA=5, InRangeB=86)) -> (Output_Get:double)
   > Set crouchAlpha(crouchAlpha:bool=Get bIsCrouching()) -> (Output_Get:bool)
   n4 = Get bIsSkyDiving()
   n5 = GetOwnerSkeletalMeshComponent()
   n6 = GetDeltaTransformFromRefPose(self=@n5, BoneName=clavicle_r, BaseName=spine_05)
   n7 = "Absolute (Float)"(A=BreakRotator(InRot=BreakTransform(InTransform=@n6).Rotation).Pitch)
   > Set clavSwingDelta_r(clavSwingDelta_r:double=float * float(A=MapRangeClamped(Value=@n7, InRangeA=0, InRangeB=30), B=SelectFloat(A=1, B=0.5, bPickA=@n4))) -> (Output_Get:double)
   n8 = GetDeltaTransformFromRefPose(self=@n5, BoneName=clavicle_l, BaseName=spine_05)
   n9 = "Absolute (Float)"(A=BreakRotator(InRot=BreakTransform(InTransform=@n8).Rotation).Pitch)
   > Set clavSwingDelta_l(clavSwingDelta_l:double=float * float(A=MapRangeClamped(Value=@n9, InRangeA=0, InRangeB=30), B=SelectFloat(A=1, B=0.5, bPickA=@n4))) -> (Output_Get:double)
   TryGetPawnOwner()
   Reroute Node(InputPin:Wildcard) -> (OutputPin:Wildcard)
   Reroute Node()
  ]}
]
variables [
 armSwingDelta_l {type=number}
 armSwingDelta_r {type=number}
 clavSwingDelta_l {type=number}
 clavSwingDelta_r {type=number}
 crouchAlpha {type=boolean}
]
