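# Physics asset for the Sparkle Remix female character's hair and face accessory mesh, defining bodies and joints for the hair's dynamic bones.
# Capsule chains cover temple locks, right bang and forehead strands, while boxes cover side and back short hair.
# Small base bodies join each piece to the head with free angular motion, and downstream joints have tighter swing and twist limits.
# Head, neck, upper spine and clavicle bodies supply the surrounding body volume.
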PhysicsAsset /BRCosmetics/Characters/Player/Female/Medium/Bodies/F_MED_SparkleRemix/Meshes/Parts/F_MED_Sparkle_Remix_New_Hair_FaceAcc_Physics
bodies {
 clavicle_l [{shapeType=Capsule, length=13.121, radius=5.5989, center=<6.7539 -0.0007 0>, rotation=<rot 90 0 0>}]
 clavicle_r [{shapeType=Capsule, length=13.0328, radius=5.687, center=<-7.1343 0.0003 0>, rotation=<rot 90 0 0>}]
 dyn_bang_00_r1 [{shapeType=Capsule, length=3.0309, radius=1.115, center=<1.9077 -0.1305 -0.1429>, rotation=<rot 87.2648 89.999 84.7774>}]
 dyn_bang_00_r1_Base [{shapeType=Capsule, length=0.505, radius=0.505, rotation=<rot 0 -90 0>}]
 dyn_bang_01_r [{shapeType=Capsule, length=3.2511, radius=0.8947, center=<1.9077 -0.1305 -0.1429>, rotation=<rot 87.2648 89.999 84.7774>}]
 dyn_foreheadA [{shapeType=Capsule, length=4.4052, radius=0.6303, center=<2.4017 -0.0854 -0.1451>, rotation=<rot 87.2659 89.9999 88.4182>}]
 dyn_foreheadA_Base [{shapeType=Capsule, length=0.505, radius=0.505, rotation=<rot 0 -90 0>}]
 dyn_foreheadB_01 [{shapeType=Capsule, length=4.3232, radius=0.7123, center=<2.4017 -0.0854 -0.1451>, rotation=<rot 87.2659 89.9999 88.4182>}]
 dyn_foreheadB_01_Base [{shapeType=Capsule, length=0.505, radius=0.505, rotation=<rot 0 -90 0>}]
 dyn_foreheadB_02 [{shapeType=Capsule, length=4.4747, radius=0.5608, center=<2.0904 0.0823 -0.153>, rotation=<rot 87.2659 89.9999 88.4182>}]
 dyn_shortHairBack_Base_l [{shapeType=Box, extentX=1.01, extentY=1.01, extentZ=1.01, rotation=<rot 0 -90 0>}]
 dyn_shortHairBack_Base_r [{shapeType=Box, extentX=1.01, extentY=1.01, extentZ=1.01, rotation=<rot 0 -90 0>}]
 dyn_shortHairBack_l [{shapeType=Box, extentX=1.01, extentY=6.7569, extentZ=5.6566, center=<1.8396 -0.0489 0.1828>, rotation=<rot 0 -90 0>}]
 dyn_shortHairBack_r [{shapeType=Box, extentX=1.01, extentY=7.0836, extentZ=5.6566, center=<2.2329 0.2141 -0.15>, rotation=<rot 0 -90 0>}]
 dyn_shortHair_Base_l [{shapeType=Box, extentX=1.01, extentY=1.01, extentZ=1.01, rotation=<rot 0 -90 0>}]
 dyn_shortHair_Base_r [{shapeType=Box, extentX=1.01, extentY=1.01, extentZ=1.01, rotation=<rot 0 -90 0>}]
 dyn_shortHair_l [{shapeType=Box, extentX=1.01, extentY=4.3851, extentZ=5.6566, center=<1.875 -0.2606 0.1825>, rotation=<rot 0 -90 0>}]
 dyn_shortHair_r [{shapeType=Box, extentX=1.01, extentY=4.3851, extentZ=5.6566, center=<1.8431 -0.1982 -0.425>, rotation=<rot 0 -90 0>}]
 dyn_templeLock_01_Base_l [{shapeType=Capsule, length=0.505, radius=0.505, rotation=<rot 87.2659 89.9999 89.9999>}]
 dyn_templeLock_01_Base_r [{shapeType=Capsule, length=0.505, radius=0.505, rotation=<rot 87.2651 89.9992 89.9992>}]
 dyn_templeLock_01_l [{shapeType=Capsule, length=4, radius=1.5, center=<2.5 0 0>, rotation=<rot 90 0 0>}]
 dyn_templeLock_01_r [{shapeType=Capsule, length=3.5, radius=1.5, center=<2.5 0 0>, rotation=<rot 90 0 0>}]
 dyn_templeLock_02_l [{shapeType=Capsule, length=3, radius=1, center=<2 0 0>, rotation=<rot 90 0 0>}]
 dyn_templeLock_02_r [{shapeType=Capsule, length=3, radius=1.25, center=<2 0 0>, rotation=<rot 90 0 0>}]
 dyn_templeLock_03_l [{shapeType=Capsule, length=3, radius=0.75, center=<2 0 0>, rotation=<rot 90 0 0>}]
 dyn_templeLock_03_r [{shapeType=Capsule, length=3.5, radius=0.75, center=<2 0 0>, rotation=<rot 90 0 0>}]
 head [
  {shapeType=Sphere, radius=6.9947, center=<4.5936 0.6314 0>}
  {shapeType=Capsule, length=10.6952, radius=5.2148, center=<0 1.5986 0.0001>, rotation=<rot -0.0001 83.422 89.9995>}
 ]
 neck_01 [{shapeType=Capsule, length=5.0595, radius=4.6557, rotation=<rot 90 0 0>}]
 spine_05 [{shapeType=Capsule, length=22.0727, radius=6.834, center=<0 1.5212 -0.0001>, rotation=<rot 0 -90 0>}]
}
constraints [
 {
  bone1Name=dyn_templeLock_01_Base_r
  bone2Name=head
  swing1LimitDegrees=45
  swing1Motion=Free
  swing2LimitDegrees=45
  swing2Motion=Free
  twistLimitDegrees=45
  twistMotion=Free
 }
 {bone1Name=dyn_templeLock_01_r, bone2Name=dyn_templeLock_01_Base_r, swing1LimitDegrees=25, swing2LimitDegrees=25, twistLimitDegrees=12}
 {bone1Name=dyn_templeLock_02_r, bone2Name=dyn_templeLock_01_r, swing1LimitDegrees=25, swing2LimitDegrees=25, twistLimitDegrees=12}
 {bone1Name=dyn_templeLock_03_r, bone2Name=dyn_templeLock_02_r, swing1LimitDegrees=25, swing2LimitDegrees=25, twistLimitDegrees=12}
 {
  bone1Name=dyn_templeLock_01_Base_l
  bone2Name=head
  swing1LimitDegrees=45
  swing1Motion=Free
  swing2LimitDegrees=45
  swing2Motion=Free
  twistLimitDegrees=45
  twistMotion=Free
 }
 {bone1Name=dyn_templeLock_01_l, bone2Name=dyn_templeLock_01_Base_l, swing1LimitDegrees=25, swing2LimitDegrees=25, twistLimitDegrees=12}
 {bone1Name=dyn_templeLock_02_l, bone2Name=dyn_templeLock_01_l, swing1LimitDegrees=25, swing2LimitDegrees=25, twistLimitDegrees=12}
 {bone1Name=dyn_templeLock_03_l, bone2Name=dyn_templeLock_02_l, swing1LimitDegrees=25, swing2LimitDegrees=25, twistLimitDegrees=12}
 {
  bone1Name=dyn_bang_00_r1_Base
  bone2Name=head
  swing1LimitDegrees=45
  swing1Motion=Free
  swing2LimitDegrees=45
  swing2Motion=Free
  twistLimitDegrees=45
  twistMotion=Free
 }
 {bone1Name=dyn_bang_00_r1, bone2Name=dyn_bang_00_r1_Base, swing1LimitDegrees=8, swing2LimitDegrees=8, twistLimitDegrees=8}
 {bone1Name=dyn_bang_01_r, bone2Name=dyn_bang_00_r1, swing1LimitDegrees=8, swing2LimitDegrees=8, twistLimitDegrees=8}
 {bone1Name=dyn_shortHair_l, bone2Name=dyn_shortHair_Base_l, swing1LimitDegrees=30, swing2LimitDegrees=20, twistLimitDegrees=8}
 {bone1Name=dyn_shortHairBack_l, bone2Name=dyn_shortHairBack_Base_l, swing1LimitDegrees=30, swing2LimitDegrees=20, twistLimitDegrees=8}
 {bone1Name=dyn_shortHairBack_r, bone2Name=dyn_shortHairBack_Base_r, swing1LimitDegrees=30, swing2LimitDegrees=20, twistLimitDegrees=8}
 {bone1Name=dyn_shortHair_r, bone2Name=dyn_shortHair_Base_r, swing1LimitDegrees=30, swing2LimitDegrees=20, twistLimitDegrees=8}
 {
  bone1Name=dyn_shortHair_Base_l
  bone2Name=head
  swing1LimitDegrees=45
  swing1Motion=Free
  swing2LimitDegrees=45
  swing2Motion=Free
  twistLimitDegrees=45
  twistMotion=Free
 }
 {
  bone1Name=dyn_shortHairBack_Base_l
  bone2Name=head
  swing1LimitDegrees=45
  swing1Motion=Free
  swing2LimitDegrees=45
  swing2Motion=Free
  twistLimitDegrees=45
  twistMotion=Free
 }
 {
  bone1Name=dyn_shortHairBack_Base_r
  bone2Name=head
  swing1LimitDegrees=45
  swing1Motion=Free
  swing2LimitDegrees=45
  swing2Motion=Free
  twistLimitDegrees=45
  twistMotion=Free
 }
 {
  bone1Name=dyn_shortHair_Base_r
  bone2Name=head
  swing1LimitDegrees=45
  swing1Motion=Free
  swing2LimitDegrees=45
  swing2Motion=Free
  twistLimitDegrees=45
  twistMotion=Free
 }
 {bone1Name=dyn_foreheadA, bone2Name=dyn_foreheadA_Base, swing1LimitDegrees=25, swing2LimitDegrees=25, twistLimitDegrees=12}
 {
  bone1Name=dyn_foreheadB_01_Base
  bone2Name=head
  swing1LimitDegrees=45
  swing1Motion=Free
  swing2LimitDegrees=45
  swing2Motion=Free
  twistLimitDegrees=45
  twistMotion=Free
 }
 {bone1Name=dyn_foreheadB_01, bone2Name=dyn_foreheadB_01_Base, swing1LimitDegrees=25, swing2LimitDegrees=25, twistLimitDegrees=12}
 {bone1Name=dyn_foreheadB_02, bone2Name=dyn_foreheadB_01, swing1LimitDegrees=25, swing2LimitDegrees=25, twistLimitDegrees=12}
 {
  bone1Name=dyn_foreheadA_Base
  bone2Name=head
  swing1LimitDegrees=45
  swing1Motion=Free
  swing2LimitDegrees=45
  swing2Motion=Free
  twistLimitDegrees=45
  twistMotion=Free
 }
 {
  bone1Name=clavicle_l
  bone2Name=spine_05
  swing1LimitDegrees=45
  swing1Motion=Free
  swing2LimitDegrees=45
  swing2Motion=Free
  twistLimitDegrees=45
  twistMotion=Free
 }
 {
  bone1Name=clavicle_r
  bone2Name=spine_05
  swing1LimitDegrees=45
  swing1Motion=Free
  swing2LimitDegrees=45
  swing2Motion=Free
  twistLimitDegrees=45
  twistMotion=Free
 }
 {
  bone1Name=head
  bone2Name=neck_01
  swing1LimitDegrees=45
  swing1Motion=Free
  swing2LimitDegrees=45
  swing2Motion=Free
  twistLimitDegrees=45
  twistMotion=Free
 }
 {
  bone1Name=neck_01
  bone2Name=spine_05
  swing1LimitDegrees=45
  swing1Motion=Free
  swing2LimitDegrees=45
  swing2Motion=Free
  twistLimitDegrees=45
  twistMotion=Free
 }
]
